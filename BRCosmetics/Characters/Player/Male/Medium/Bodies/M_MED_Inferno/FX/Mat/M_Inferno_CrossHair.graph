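# Unlit translucent Niagara sprite material that draws a tinted cross-flare for the Inferno character body's crosshair effect.
# A grayscale flare texture blends two cross colors through an inverted power curve, then particle color tints it before the emissive adjustments function.
# The same texture, shaped by MF_FX_Refine and multiplied by particle alpha, feeds a depth fade for soft intersections.
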Material /BRCosmetics/Characters/Player/Male/Medium/Bodies/M_MED_Inferno/FX/Mat/M_Inferno_CrossHair
graph {
 properties {bUsedWithNiagaraSprites=true, blendMode=BLEND_Translucent, pixelDepthOffsetMode=PDOM_Legacy, shadingModel=MSM_Unlit, usedShadingModels=Unlit}
 nodes [
  n1 = ParticleColor()
  n2 = TextureSample() {samplerType=SAMPLERTYPE_Grayscale, texture=/Game/Effects/Fort_Effects/Textures/Flares/T_Cross_Flare}
  n3 = VectorParameter() {parameterName=Refine, defaultValue=<rgba 3 3 0.6 0.5>}
  DepthFade(
   Opacity=Multiply(
    A=@n1.A
    B=MaterialFunctionCall(Texture=@n2.RGB, RGB=@n3.RGB, A=@n3.RGB) {materialFunction=&MF_FX_Refine}.Result
   )
   FadeDistance=ScalarParameter() {defaultValue=2, parameterName=DF}
  )
  n4 = Power(Base=OneMinus(None=Saturate(None=@n2.RGB)), Exp=ScalarParameter() {defaultValue=3.25, parameterName=test})
  n5 = LinearInterpolate(
   A=VectorParameter() {parameterName=CrossColor2, defaultValue=<rgba 3.4667 0 13 1>}.RGB
   B=Multiply(A=Add(A=@n2.RGB), B=VectorParameter() {parameterName=CrossColor, defaultValue=<rgba 0 0.8333 5 1>}.RGB)
   Alpha=Saturate(None=OneMinus(None=@n4))
  )
  MaterialFunctionCall(Emissive=Divide(A=Multiply(A=@n1.RGB, B=@n5)) {constB=1.977}) {materialFunction=&MF_FX_EmissiveAdjustments}
 ]
}
