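# Translucent unlit material for Niagara mesh particles in the Inferno character's FX, named as a temporary center emissive.
# A curve atlas sampled along U, or a solid TargetColor by static switch, gives color, boosted by the RootEmissive dynamic parameter before MF_FX_EmissiveAdjustments.
# An opacity chain blends curve alpha with an optional V edge mask, then applies ScaleOpacity and a second V edge mask.
# The Fix lerp has identical inputs, so does nothing.
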Material /BRCosmetics/Characters/Player/Male/Medium/Bodies/M_MED_Inferno/FX/Mat/M_Inferno_TempCenterEmissive
graph {
 properties {
  bUsedWithNiagaraMeshParticles=true
  blendMode=BLEND_Translucent
  pixelDepthOffsetMode=PDOM_Legacy
  shadingModel=MSM_Unlit
  usedShadingModels=Unlit
 }
 nodes [
  n1 = MaterialFunctionCall(Float2=TextureCoordinate()) {materialFunction=&BreakOutFloat2Components}
  n2 = CurveAtlasRowParameter(CurveTime=@n1.R) {
   atlas=/Game/Characters/Player/Male/Medium/Bodies/M_MED_Inferno/FX/Mat/CLA_InfernoCurve
   curve=/Game/Characters/Player/Male/Medium/Bodies/M_MED_Inferno/FX/Mat/CL_Inferno_Base
   defaultValue=1
   parameterName=A
  }
  n3 = StaticSwitchParameter(True=VectorParameter() {parameterName=TargetColor, defaultValue=<rgba 1 1 1 1>}.RGB, False=@n2) {parameterName=UseSolidColor}
  n4 = ScalarParameter() {defaultValue=2.5, parameterName=RootEmStr}
  n5 = DynamicParameter() {parameterIndex=2, paramNames [Param1,Param2,RootEmissive,Param4]}
  MaterialFunctionCall(Emissive=Multiply(A=Divide(A=@n3) {constB=1.977}, B=LinearInterpolate(B=@n4, Alpha=@n5.RootEmissive) {constA=1})) {materialFunction=&MF_FX_EmissiveAdjustments}
  Multiply(A=VertexNormalWS(), B=ScalarParameter() {defaultValue=0.1, parameterName=NormalStr})
  n6 = VectorParameter() {
   group=TipAndBackOpacity
   parameterName=TipAndBackOpacity_EdgeMask_InvertControl
   channelNames {a="", b=Power, g=Multiply, r=Offset}
   defaultValue=<rgba 0 10 2 1>
  }
  n7 = VectorParameter() {
   group=TipAndBackOpacity
   parameterName=TipAndBackOpacity_EdgeMask_Control
   channelNames {a="", b=Power, g=Multiply, r=Offset}
   defaultValue=<rgba 0 15 2 1>
  }
  n8 = ScalarParameter() {parameterName=Fix}
  n9 = VectorParameter() {group=Opacity, parameterName=Opacity_EdgeMask_Control, channelNames {a="", b=Power, g=Multiply, r=Offset}, defaultValue=<rgba 0 1 1 1>}
  n10 = Multiply(
   A=@n2.A
   B=MaterialFunctionCall!(UorV=LinearInterpolate(A=@n1.G, B=@n1.G, Alpha=@n8), EdgeControl=@n9) {materialFunction=/Game/Effects/Niagara/MasterMaterials/Functions/Common/FXCommon_EdgeMask_Simple}.Mask
  )
  n11 = ScalarParameter() {parameterName=UseL}
  Multiply(
   A=MaterialFunctionCall!(UorV=@n1.G, Invert_EdgeControl=@n6, NoInvert_EdgeControl=@n7) {materialFunction=/Game/Effects/Niagara/MasterMaterials/Functions/Common/FXCommon_EdgeMask_Ex}.Mask
   B=Multiply(A=LinearInterpolate(A=@n2.A, B=@n10, Alpha=@n11), B=ScalarParameter() {defaultValue=1, parameterName=ScaleOpacity})
  )
  VectorParameter() {parameterName=Color, defaultValue=<rgba 0 0.7721 1 1>}
 ]
}
