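# Material function that rotates vertices about an axis and pivot while keeping the shading basis consistent, stored with the Inferno character's FX materials.
# RotateAboutAxis turns Position (default: world position excluding shader offsets) into the WPO output.
# Normal, TangentU and TangentV get the same rotation about a zero pivot, with each offset added back and normalized.
# Those three inputs default to unit axes transformed from tangent space.
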MaterialFunction /BRCosmetics/Characters/Player/Male/Medium/Bodies/M_MED_Inferno/FX/Mat/RotateVertexAboutAxis
graph {nodes [
  n1 = FunctionInput() {inputName=NormalizedRotationAxis}
  n2 = FunctionInput() {inputName=RotationAngle, inputType=FunctionInput_Scalar}
  n3 = WorldPosition() {worldPositionShaderOffset=WPT_ExcludeAllShaderOffsets}
  FunctionOutput(None=RotateAboutAxis(
    NormalizedRotationAxis=@n1
    RotationAngle=@n2
    PivotPoint=FunctionInput() {inputName=PivotPoint, sortPriority=2}
    Position=FunctionInput(Preview=@n3.XYZ) {bUsePreviewValueAsDefault=true, inputName=Position, sortPriority=3}
   )) {outputName=WPO}
  n4 = FunctionInput(Preview=Transform(None=Constant3Vector() {constant=<rgba 0 0 1 0>})) {bUsePreviewValueAsDefault=true, inputName=Normal, sortPriority=10}
  n5 = Constant3Vector()
  n6 = RotateAboutAxis(NormalizedRotationAxis=@n1, RotationAngle=@n2, PivotPoint=@n5, Position=@n4)
  FunctionOutput(None=Normalize(VectorInput=Add(A=@n4, B=@n6))) {outputName=Normal, sortPriority=1}
  n7 = FunctionInput(Preview=Transform(None=Constant3Vector() {constant=<rgba 0 1 0 0>})) {bUsePreviewValueAsDefault=true, inputName=TangentV, sortPriority=12}
  n8 = RotateAboutAxis(NormalizedRotationAxis=@n1, RotationAngle=@n2, PivotPoint=@n5, Position=@n7)
  FunctionOutput(None=Normalize(VectorInput=Add(A=@n7, B=@n8))) {outputName=TangentV, sortPriority=3}
  n9 = FunctionInput(Preview=Transform(None=Constant3Vector() {constant=<rgba 1 0 0 0>})) {bUsePreviewValueAsDefault=true, inputName=TangentU, sortPriority=11}
  n10 = RotateAboutAxis(NormalizedRotationAxis=@n1, RotationAngle=@n2, PivotPoint=@n5, Position=@n9)
  FunctionOutput(None=Normalize(VectorInput=Add(A=@n9, B=@n10))) {outputName=TangentU, sortPriority=2}
 ]}
